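# Control Rig for the Olivia face accessory part that adjusts the hat bone as the head moves.
# A spherical pose reader measures the head bone's orientation with spine_01 as optional parent, using an asymmetric active region and falloff.
# Its output weights a Modify Transforms step that shifts and tilts dyn_olivia_hat.
# Debug drawing is enabled on the reader, and the hair bones in the hierarchy are not driven.
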ControlRigBlueprint /BRCosmetics/Characters/Player/Female/Medium/Bodies/F_MED_Olivia/Meshes/Parts/F_MED_Olivia_FaceAcc_CtrlRig
graphs [
 RigVMFunctionLibrary {links [], nodes []}
 RigVMModel {nodes [
   RigUnit_BeginExecution = &RigUnit_BeginExecution()
   n1 = &RigUnit_SphericalPoseReader(
    ActiveRegionScaleFactors=(PositiveWidth=3,NegativeWidth=0.1,PositiveHeight=6,NegativeHeight=6)
    ActiveRegionSize=0.1
    Debug=(bDrawDebug=true,DebugScale=25,DebugSegments=20,DebugThickness=0.25)
    DriverAxis=(X=1,Y=0,Z=0)
    DriverItem=(Type=Bone,Name="head")
    FalloffRegionScaleFactors=(PositiveWidth=-3,NegativeWidth=1,PositiveHeight=0,NegativeHeight=0)
    FalloffSize=0.25
    OptionalParentItem=(Type=Bone,Name="spine_01")
    RotationOffset=(X=90,Y=0,Z=-180)
    ExecutePin=@RigUnit_BeginExecution.ExecutePin
   )
   > &RigUnit_ModifyTransforms(
    ItemToModify=((Item=(Type=Bone,Name="dyn_olivia_hat"),Transform=(Rotation=(X=0,Y=-0.1305,Z=0,W=0.9914),Translation=(X=0,Y=0,Z=5),Scale3D=(X=1,Y=1,Z=1))))
    Weight=@n1.OutputParam
   )
  ]}
]
hierarchy [
 root {loc=<0 0 0>, rot=<0 0 0 1>}
 pelvis {parent=root, loc=<0 -0.5301 98.0571>, rot=<0.0127 0.707 -0.0127 -0.707>}
 spine_01 {parent=pelvis, loc=<6.1208 -0.9328 0>, rot=<0 0 0.034 -0.9994>}
 spine_02 {parent=spine_01, loc=<6.5996 -0.2095 0>, rot=<0 0 0.0238 -0.9997>}
 spine_03 {parent=spine_02, loc=<6.4914 -0.1237 0>, rot=<0 0 0.0149 -0.9999>}
 spine_04 {parent=spine_03, loc=<6.9598 0.0794 0>, rot=<0 0 -0.0132 -0.9999>}
 spine_05 {parent=spine_04, loc=<6.9987 0 0>, rot=<0 0 -0.0015 -1>}
 neck_01 {parent=spine_05, loc=<11.3138 0.1529 0>, rot=<0 0 -0.2003 -0.9797>}
 neck_02 {parent=neck_01, loc=<4.5627 0 0>, rot=<0 0 0.0504 -0.9987>}
 head {parent=neck_02, loc=<5.3914 0 0>, rot=<0 0 0.115 -0.9934>}
 base_hair_0 {parent=head, loc=<9.8562 8.3973 -0.6866>, rot=<-0.1499 -0.691 0.6957 0.1263>}
 dyn_hair_0 {parent=base_hair_0, loc=<0 0 0>, rot=<-0.0001 -0.0001 0.0001 -1>}
 dyn_hair_1 {parent=dyn_hair_0, loc=<2.4537 0.0003 -0.0002>, rot=<0 0 -0.2504 -0.9681>}
 dyn_hair_2 {parent=dyn_hair_1, loc=<2.5547 0.0001 -0.0004>, rot=<0.0001 0 0.0001 -1>}
 base_olivia_hat {parent=head, loc=<3.7351 -7.3322 0.2525>, rot=<0.4408 -0.5529 0.5529 -0.4408>}
 dyn_olivia_hat {parent=base_olivia_hat, loc=<0 0 0>, rot=<0 0 0 1>}
]
