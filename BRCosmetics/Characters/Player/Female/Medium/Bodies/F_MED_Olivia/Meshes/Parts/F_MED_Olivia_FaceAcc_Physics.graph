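# Physics asset for the face accessory mesh of the Olivia female player cosmetic body, defining physics for a short chain of hair bones.
# A small base capsule on base_hair_0 anchors three identical capsules on dyn_hair_0 through dyn_hair_2, linked in series by constraints.
# Each joint has zero twist and narrow swing limits, with the secondary swing widening toward the tip.
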PhysicsAsset /BRCosmetics/Characters/Player/Female/Medium/Bodies/F_MED_Olivia/Meshes/Parts/F_MED_Olivia_FaceAcc_Physics
bodies {
 base_hair_0 [{shapeType=Capsule, length=0.505, radius=0.505, shapeName=base_hair_0_capsule, rotation=<rot 0 -90 0>}]
 dyn_hair_0 [{shapeType=Capsule, length=2, radius=1, shapeName=dyn_hair_0_capsule, center=<2 0 0>, rotation=<rot 90 0 0>}]
 dyn_hair_1 [{shapeType=Capsule, length=2, radius=1, shapeName=dyn_hair_1_capsule, center=<2 0 0>, rotation=<rot 90 0 0>}]
 dyn_hair_2 [{shapeType=Capsule, length=2, radius=1, shapeName=dyn_hair_2_capsule, center=<2 0 0>, rotation=<rot 90 0 0>}]
}
constraints [
 {bone1Name=dyn_hair_0, bone2Name=base_hair_0, swing1LimitDegrees=10, swing2LimitDegrees=5, twistLimitDegrees=0}
 {bone1Name=dyn_hair_1, bone2Name=dyn_hair_0, swing1LimitDegrees=10, swing2LimitDegrees=8, twistLimitDegrees=0}
 {bone1Name=dyn_hair_2, bone2Name=dyn_hair_1, swing1LimitDegrees=10, swing2LimitDegrees=10, twistLimitDegrees=0}
]
